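# GPU Niagara atmosphere for the FNM2025 MtFetch area, layering bats, fog and airborne leaves, all responsive to time of day.
# Capped mesh bats orbit a raised vortex, avoiding the camera and distance-field surfaces.
# Sprite and mesh fog samples the landscape for lift, shrinks up close and is culled inside hard-coded world boxes.
# Leaves spawn via depth-buffer projection and GBuffer hue filtering, with wind gusts boosting spawn rate and drift.
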NiagaraSystem /Game/Effects/Fort_Effects/Effects/Environment/FNM2025/MtFetch/NS_FNM_Gourdy_Atmosphere
dependencies {
 usedDataInterfaces [
  &NiagaraDataInterfaceCurve
  &NiagaraDataInterfaceCamera
  &NiagaraDataInterfaceGBuffer
  &NiagaraDataInterfaceArrayFloat
  &NiagaraDataInterfaceCollisionQuery
 ]
 usedDynamicInputs [&SimulationPosition]
 usedModules [&SystemState]
}
emitters {
 Bats001 {
  simTarget=GPUComputeSim
  emitterSpawnScript [/SetVariables(Emitter.PositionOffset:NiagaraPosition=<0 0 5750>, Emitter.RadiusMult=1.35) {isSetParameters=true}]
  emitterUpdateScript [
   &EmitterState(Life Cycle Mode=::NewEnumerator1, Loop Behavior=::NewEnumerator1, Loop Duration Mode=::NewEnumerator1, UseLoopDelay=false)
   &SpawnRate(
    Spawn Probability=&FloatFromCurve(
     FloatCurve=DataInterface() {propertyValues {Curve {
        defaultValue=1
        keys [
         {interpMode=RCIM_Cubic, tangentMode=RCTM_User, time=0, value=1}
         {arriveTangent=-2, interpMode=RCIM_Cubic, leaveTangent=-0.3147, tangentMode=RCTM_Break, time=0.3, value=0}
         {interpMode=RCIM_Cubic, tangentMode=RCTM_User, time=0.7, value=0}
         {tangentMode=RCTM_User, time=1, value=1}
        ]
       }}}
     CurveIndex=Linked() {linkedVariable {name=NPC.FortniteNPC.TimeOfDay_0-1, type=NiagaraFloat}}
    )
    SpawnRate=&RandomRangeFloat(Minimum=5, Maximum=9)
    Use Spawn Probability=true
   )
   /SetVariables(Emitter.CameraPosition:Vector3f=&ReturnCameraPosition(Coordinate Space:ENiagaraCoordinateSpace=::Simulation, Camera Query:NiagaraDataInterfaceCamera=DataInterface())) {isSetParameters=true}
  ]
  particleSpawnScript [
   &InitializeParticle(
    Color=<rgba 3 3 3 1>
    Lifetime=&RandomRangeFloat(Minimum=8, Maximum=10)
    Mesh Scale Mode=::NewEnumerator3
    Mesh Uniform Scale=&Multiply_Float(A=&RandomRangeFloat(Minimum=0.15, Maximum=0.3), B=2)
    UsePositionOffset=false
    Write Material Random=true
   )
   &ShapeLocation(
    Enable Torus Knot Radius=false
    Handle Radius:NiagaraFloat=&Multiply_Float(A=300, B=$Emitter.RadiusMult)
    Large Radius=&Multiply_Float(A=&RandomRangeFloat(Minimum=1500, Maximum=5500), B=$Emitter.RadiusMult)
    MaskSpawnGroup=false
    Offset:NiagaraPosition=$Emitter.PositionOffset
    Offset Mode=::NewEnumerator0
    Override Seed=false
    Shape Primitive=::NewEnumerator3
    Torus Mode:ENiagaraShapeTorusMode=::NewEnumerator1
    UseOwnerScale=true
   )
   &KillParticles(Kill Particles=&SetBoolByFloatComparison(
     A=&Divide_Float(A=&MakeFloatFromInt(Int=$Engine.Emitter.NumParticles), B=32)
     B=1
    )) {module=KillParticles001}
   /SetVariables(Particles.KilledBYSpawnLimit?:NiagaraBool=&SetBoolByFloatComparison(
     A=&Divide_Float(A=&MakeFloatFromInt(Int=$Engine.Emitter.NumParticles), B=32)
     B=1
    )) {isSetParameters=true}
  ]
  particleUpdateScript [
   &ParticleState()
   /SetVariables(Particles.RandomDirection:Vector3f=&Lerp_Vector(
     A=&MakeVector(
      X=&Waveform(
       Global Phase=&Add_Float(A=$Particles.Age, B=&RandomRangeFloat(Minimum=0, Maximum=5))
       "Waveform [1]"=::NewEnumerator8
       "[1] Frequency"=1
       "[1] Amplitude Min/Max"=<-1 1>
       "[2] Phase"=$Emitter.Age
       "[3] Phase"=$Emitter.Age
      )
      Y=&Waveform(
       Global Phase=&Add_Float(A=$Particles.Age, B=&RandomRangeFloat(Minimum=0, Maximum=5))
       "Waveform [1]"=::NewEnumerator8
       "[1] Frequency"=1
       "[1] Amplitude Min/Max"=<-1 1>
       "[2] Phase"=$Emitter.Age
       "[3] Phase"=$Emitter.Age
      )
      Z=0
     )
     B=<1 0 0>
     Alpha=&Waveform(
      Global Phase=&Add_Float(A=$Particles.Age, B=&RandomRangeFloat(Minimum=0, Maximum=5))
      Global Clamp Min=0.4
      "Waveform [1]"=::NewEnumerator8
      "[1] Frequency"=1
      "[1] Amplitude Min/Max"=<0 1>
      "[2] Phase"=$Emitter.Age
      "[3] Phase"=$Emitter.Age
     )
    )) {enabled=false, isSetParameters=true}
   &InterpolateOverTime(
    Target Vector Value:Vector3f=&NormalizeVector(Vector To Normalize=$Particles.RandomDirection, Safe Normalize=true, Safe Fallback Threshold=0, Safe Fallback Vector=<0 0 0.001>)
    Variable Type=::NewEnumerator2
   ) {enabled=false, module=InterpolateOverTime001}
   &WindForce(
    Air Resistance=0.1
    Wind Speed=&TransformVector(Vector=$Particles.InterpolateOverTime001.InterpolateOverTime_V3_.Moving Average, Destination Space=::Simulation)
    Wind Speed Scale=&RandomRangeFloat(Minimum=100, Maximum=200)
   ) {enabled=false}
   &PointAttractionForce(
    Attraction Radius=10000
    Attraction Strength=&RandomRangeFloat(Minimum=3, Maximum=6)
    Attractor Position:Vector3f=$Particles.FishMoverPosition
    Kill Within Radius=false
    Module.Use Falloff=true
    UseAttractorOffset=false
   ) {enabled=false}
   &PointAttractionForce(
    Attraction Radius=500
    Attraction Strength=-200
    Attractor Position:Vector3f=$Emitter.CameraPosition
    Kill Within Radius=false
    Module.Use Falloff=true
    UseAttractorOffset=false
   ) {module=PointAttractionForce001}
   &VortexForce(
    Origin Pull Amount:NiagaraFloat=&RandomRangeFloat(Minimum=50, Maximum=120)
    Origin Pull Only On Vortex Axis=false
    UseOriginOffset=true
    Vortex Axis=<0 0 4>
    Vortex Force Amount=100
    Vortex Origin:NiagaraPosition=&AddVectorToPosition(Position=$Engine.Emitter.SimulationPosition, Vector=Linked() {linkedVariable {name=Emitter.PositionOffset, type=NiagaraPosition}})
    Vortex Origin Offset:Vector3f=&RandomRangeVector(Minimum=<0 0 222>, Maximum=<0 0 -222>)
   )
   &VortexForce(
    Origin Pull Amount:NiagaraFloat=&RandomRangeFloat(Minimum=1, Maximum=800)
    UseOriginOffset=false
    Vortex Force Amount=&FloatFromCurve(
     FloatCurve=DataInterface() {propertyValues {Curve {keys [
         {tangentMode=RCTM_User, time=0, value=1}
         {arriveTangent=-7.824, interpMode=RCIM_Cubic, leaveTangent=-7.824, tangentMode=RCTM_User, time=0.4, value=1}
         {interpMode=RCIM_Cubic, leaveTangent=-8.0089, time=1, value=0}
        ]}}}
     Scale Curve=222
    )
   ) {enabled=false, module=VortexForce001}
   /Niagara/Modules/Math/AvoidDistanceFieldSurfaces_GPU(Avoid Nearest Surface=true, Avoid Oncoming Surfaces=true, Nearest Surface Avoidance Strength=800, Obstactle Avoidance Strength=800)
   &Collision(
    Age Colliding Particles=false
    GPU Collision Type=::NewEnumerator2
    KillOnCollision=false
    Module.Control Rest State Via Surface Normal=false
    Module.EnableMaxCollisionCount=false
    Module.Randomize Collision Normal Vector=false
    Radius Calculation Type=::NewEnumerator1
    Restitution=1
   )
   &CurlNoiseForce(Bias Noise=false, Noise Frequency=19.1141, Noise Strength=995.4374, Pan Noise=false) {enabled=false}
   &CurlNoiseForce(
    Bias Noise=false
    Noise Frequency=&UniformAOrBFloat(
     A=0
     B=&RandomRangeFloat(Minimum=19, Maximum=5)
     Distribution Weight=0.15
     Evaluation Type=::NewEnumerator0
     Recalculate Random Each Loop=true
     Override Seed=false
    )
    Noise Strength=&UniformAOrBFloat(
     A=0
     B=&RandomRangeFloat(Minimum=555, Maximum=955)
     Distribution Weight=0.15
     Evaluation Type=::NewEnumerator0
     Recalculate Random Each Loop=true
     Override Seed=false
    )
    Pan Noise=true
    Pan Noise Field=<1 0 0>
   ) {module=CurlNoiseForce001}
   /SetVariables(Particles.RandomPosition:NiagaraPosition=&Lerp_Vector(
     A=Linked() {linkedVariable {name=Particles.ShapeLocation.ShapePosition, type=NiagaraPosition}}
     B=&MakeVector(X=0, Y=0, Z=0)
     Alpha=&FloatFromCurve(FloatCurve=DataInterface() {propertyValues {Curve {
         defaultValue=1
         keys [{interpMode=RCIM_Cubic, tangentMode=RCTM_User, time=0.75, value=1}, {interpMode=RCIM_Cubic, tangentMode=RCTM_User, time=0.8, value=0}]
        }}})
    )) {enabled=false, isSetParameters=true}
   &SolveForcesAndVelocity(Clamp Velocity=true, Limit Acceleration=false, Speed Limit=600)
   /SetVariables(Particles.New Vector 3f:Vector3f=&Multiply_VectorByFloat(
     Vector=$Particles.Scale
     Float=&FloatFromCurve(FloatCurve=DataInterface() {propertyValues {Curve {keys [
          {interpMode=RCIM_Cubic, tangentMode=RCTM_User, time=0, value=0}
          {interpMode=RCIM_Cubic, tangentMode=RCTM_User, time=0.1, value=1}
          {interpMode=RCIM_Cubic, tangentMode=RCTM_User, time=0.8, value=1}
          {tangentMode=RCTM_User, time=1, value=0}
         ]}}})
    )) {isSetParameters=true}
   &KillParticlesInVolume(Invert Volume=true, Sphere Radius=2000) {enabled=false}
   &DynamicMaterialParameters(
    Index 0 Param 1=0.25
    Index 0 Param 2=&RandomRangeFloat(Minimum=2.8, Maximum=3.5)
    Index 0 Param 3=0.25
    Index 0 Param 4=0.1
    Index 1 Param 1=1
    Index 1 Param 2=&RandomRangeFloat(Minimum=25, Maximum=55)
    Index 1 Param 3=2
    Index 1 Param 4=0.25
    Index 2 Param 1=1
    Index 2 Param 2=1
    Index 2 Param 3=2
    Index 2 Param 4=0.25
    Module.Index0Param0Write=true
    Module.Index0Param1Write=true
    Module.Index0Param2Write=true
    Module.Index0Param3Write=true
    Module.Index1Param0Write=true
    Module.Index1Param1Write=true
    Module.Index1Param2Write=true
    Module.Index1Param3Write=true
    Module.Index2Param0Write=true
    Module.Index2Param1Write=true
    Module.Index2Param2Write=true
    Module.Index2Param3Write=true
    Write Parameter Index 1=true
    Write Parameter Index 2=true
   ) {module=DynamicMaterialParameters001}
  ]
  rendererClasses [&NiagaraMeshRendererProperties]
  renderers [{rendererClass=&NiagaraMeshRendererProperties, rendererIndex=0}]
 }
 Fog {
  simTarget=GPUComputeSim
  emitterSpawnScript [/SetVariables(Emitter.ScaleColorAlpha=20, Emitter.ScaleMult:NiagaraFloat=&RandomRangeFloat(Minimum=2, Maximum=2.5)) {isSetParameters=true}]
  emitterUpdateScript [
   &EmitterState()
   &SpawnRate(
    SpawnRate=&FloatFromCurve(
     FloatCurve=DataInterface() {propertyValues {Curve {
        defaultValue=1
        keys [
         {interpMode=RCIM_Cubic, tangentMode=RCTM_User, time=0, value=1}
         {interpMode=RCIM_Cubic, tangentMode=RCTM_User, time=0.5, value=0}
         {tangentMode=RCTM_User, time=1, value=1}
        ]
       }}}
     CurveIndex=Linked() {linkedVariable {name=NPC.FortniteNPC.TimeOfDay_0-1, type=NiagaraFloat}}
     Scale Curve=12
    )
    Use Spawn Probability=false
   ) {enabled=false, module=SpawnRate001}
   &SpawnRate(SpawnRate=12, Use Spawn Probability=false)
  ]
  particleSpawnScript [
   &InitializeParticle(
    Color=<rgba 42.9274 59.0472 125 0.5>
    Lifetime=&RandomRangeFloat(Minimum=4, Maximum=6)
    Mesh Scale Mode=::NewEnumerator3
    Mesh Uniform Scale=&Multiply_Float(A=&RandomRangeFloat(Minimum=4, Maximum=9), B=$Emitter.ScaleMult)
    Sprite Size Mode=::NewEnumerator3
    Uniform Sprite Size=&Multiply_Float(A=&RandomRangeFloat(Minimum=500, Maximum=750), B=$Emitter.ScaleMult)
    UsePositionOffset=false
    Write Material Random=true
   ) {module=InitializeParticle002}
   &ShapeLocation(
    MaskSpawnGroup=false
    Non Uniform Scale=<1.6 0.5 1>
    Offset Mode=::NewEnumerator0
    Override Seed=false
    Sphere Radius=5500
    UseOwnerScale=true
   ) {module=ShapeLocation001}
   &KillParticles()
   &Color(Color=&RandomRangeLinearColor(Minimum=<rgba 0.0865 0.0648 0.0343 1>, Maximum=<rgba 0.1529 0.0908 0.0319 1>)) {enabled=false, module=Color001}
  ]
  particleUpdateScript [
   &ParticleState()
   &self:SampleLandscape_01() {module=SampleLandscape_01001}
   /SetVariables(Particles.Position=&AddVectorToPosition(
     Position=$Particles.Position
     Vector=&Multiply_Vector(A=$Output.SampleLandscape_01001.LandscapeNormal, B=<1 1 200>)
    )) {isSetParameters=true}
   &KillParticlesInVolume(
    Box Size:Vector3f=<3000 7000 3000>
    Kill Shape=::NewEnumerator1
    Volume Origin:Vector3f=&AddVectorToPosition(Position=<-79152 -29616 5540>, Vector=<0 0 0>)
   ) {enabled=false}
   &KillParticlesInVolume(
    Box Size:Vector3f=<6000 7000 4000>
    Kill Shape=::NewEnumerator1
    Volume Origin:Vector3f=&AddVectorToPosition(Position=<-80932 -29612 5172>, Vector=<0 0 0>)
   ) {module=KillParticlesInVolume002}
   &KillParticlesInVolume(
    Box Size:Vector3f=<5000 4000 4000>
    Kill Shape=::NewEnumerator1
    Volume Origin:Vector3f=&AddVectorToPosition(Position=<-66204 -28636 4608>, Vector=<0 0 0>)
   ) {module=KillParticlesInVolume003}
   &ScaleMeshSize(Scale Factor=&Lerp_Vector(
     A=<0.125 0.125 0.125>
     B=<1 1 1>
     Alpha=&FN_LOD_Distance_Remap(Start Distance=0, End Distance=3500, Minimum Output=0, Maximum Output=1, Clamp Output=true)
    ))
   &ScaleSpriteSize(
    Uniform Curve Index=&FN_LOD_Distance_Remap(Start Distance=0, End Distance=1500, Minimum Output=0, Maximum Output=1, Clamp Output=true)
    Uniform Curve Sprite Scale=DataInterface() {propertyValues {Curve {keys [{interpMode=RCIM_Cubic, tangentMode=RCTM_Break, time=0, value=0}, {tangentMode=RCTM_User, time=1, value=1}]}}}
   )
   &ScaleColor(
    Scale Alpha=&Multiply_Float(
     A=&Multiply_Float(
      A=&FloatFromCurve(
       FloatCurve=DataInterface() {propertyValues {Curve {keys [
           {interpMode=RCIM_Cubic, tangentMode=RCTM_User, time=0, value=0}
           {interpMode=RCIM_Cubic, tangentMode=RCTM_User, time=0.5, value=1}
           {tangentMode=RCTM_User, time=1, value=0}
          ]}}}
       Scale Curve=&Multiply_Float(
        A=&Multiply_Float(A=&RandomRangeFloat(Minimum=0.15, Maximum=0.17), B=0.7)
        B=$Emitter.ScaleColorAlpha
       )
      )
      B=&FloatFromCurve(
       FloatCurve=DataInterface() {propertyValues {Curve {
          defaultValue=1
          keys [
           {interpMode=RCIM_Cubic, tangentMode=RCTM_User, time=0, value=1.2}
           {arriveTangent=2.3353, interpMode=RCIM_Cubic, leaveTangent=-0.0832, tangentMode=RCTM_Break, time=0.25, value=1.5}
           {interpMode=RCIM_Cubic, tangentMode=RCTM_User, time=0.5, value=1.2}
           <at 0.75 1 RCIM_Cubic>
           {tangentMode=RCTM_User, time=1, value=1.2}
          ]
         }}}
       CurveIndex=Linked() {linkedVariable {name=NPC.FortniteNPC.TimeOfDay_0-1, type=NiagaraFloat}}
      )
     )
     B=&FloatFromCurve(
      FloatCurve=DataInterface() {propertyValues {Curve {
         defaultValue=0
         keys [
          {arriveTangent=1.6813, interpMode=RCIM_Cubic, leaveTangent=1.6813, tangentMode=RCTM_Break, time=0, value=0.05}
          {interpMode=RCIM_Cubic, tangentMode=RCTM_User, time=1, value=0.6}
         ]
        }}}
      CurveIndex=&RemapRange(
       Input Value To Remap=&VectorLength(Vector=&Subtract_Position(A=$Engine.Owner.Position, B=$Particles.Position))
       Input Min=0
       Input Max=10000
       Output Min=0
       Output Max=0.08
      )
      Scale Curve=&RandomRangeFloat(Minimum=0.9, Maximum=1)
     )
    )
    ScaleA=true
    ScaleRGB=false
   ) {enabled=false, module=ScaleColor001}
   &ScaleColor(
    Scale Alpha=&FloatFromCurve(
     FloatCurve=DataInterface() {propertyValues {Curve {keys [
         {interpMode=RCIM_Cubic, tangentMode=RCTM_User, time=0, value=0}
         {interpMode=RCIM_Cubic, tangentMode=RCTM_User, time=0.5, value=1}
         {tangentMode=RCTM_User, time=1, value=0}
        ]}}}
     Scale Curve=0.1
    )
    Scale RGB=&VectorFromFloat(Value=&FloatFromCurve(
      FloatCurve=DataInterface() {propertyValues {Curve {
         defaultValue=1
         keys [
          {interpMode=RCIM_Cubic, tangentMode=RCTM_User, time=0.3, value=1}
          {interpMode=RCIM_Cubic, tangentMode=RCTM_User, time=0.5, value=0.009}
          {tangentMode=RCTM_User, time=0.7, value=1}
         ]
        }}}
      CurveIndex=Linked() {linkedVariable {name=NPC.FortniteNPC.TimeOfDay_0-1, type=NiagaraFloat}}
     ))
    ScaleA=true
    ScaleRGB=true
   )
   &AccelerationForce(Acceleration=<150 150 25>)
   &Drag(Use Linear Drag=true, Use Rotational Drag=true)
   &SolveForcesAndVelocity(Clamp Velocity=false, Limit Acceleration=false)
   &CameraOffset(Camera Offset Amount=20) {enabled=false}
   &CameraOffset(Camera Offset Amount=90) {module=CameraOffset001}
  ]
  rendererClasses [&NiagaraSpriteRendererProperties, &NiagaraMeshRendererProperties]
  renderers [
   {rendererClass=&NiagaraSpriteRendererProperties}
   {rendererClass=&NiagaraMeshRendererProperties, rendererIndex=1}
   {rendererClass=&NiagaraSpriteRendererProperties, rendererIndex=2}
  ]
 }
 LeavesInAir001 {
  simTarget=GPUComputeSim
  emitterUpdateScript [
   /SetVariables(
    Emitter.Emitter_HeightOffset:NiagaraFloat=$System.System_HeightOffset
    Emitter.Overall_SpeedMult=0.2
    Emitter.SpawnRate_Curve:NiagaraFloat=&FloatFromCurve(
     FloatCurve=DataInterface() {propertyValues {Curve {keys [{arriveTangent=17.5837, time=0, value=0}, <at 0.1 1>, <at 0.2 0.0925>, <at 0.6 0.1525>, <at 0.7 0>, <at 1 0>]}}}
     CurveIndex=$Emitter.NormalizedLoopAge
     Scale Curve=100
    )
    Emitter.SpawnSpeed_Curve:NiagaraFloat=&FloatFromCurve(
     FloatCurve=DataInterface() {propertyValues {Curve {keys [
         {arriveTangent=30.2685, interpMode=RCIM_Cubic, leaveTangent=30.2685, tangentMode=RCTM_User, time=0.0009, value=0.0215}
         {arriveTangent=-0.0619, interpMode=RCIM_Cubic, leaveTangent=-0.0619, tangentMode=RCTM_User, time=0.0989, value=0.3515}
         <at 0.297 0.0032>
         <at 0.3192 0.3332>
         <at 0.3525 0.0398>
         <at 0.6171 0.3148>
         <at 0.9939 0.3698>
        ]}}}
     CurveIndex=&Multiply_Float(A=$Emitter.NormalizedLoopAge, B=&RandomRangeFloat(Minimum=1, Maximum=5))
     Scale Curve=&Multiply_Float(
      A=&Lerp_Float(A=5, B=10, Alpha=&UniformRangedFloat(Minimum=0.25, Maximum=1))
      B=1
     )
    )
    Emitter.SpriteRotRate:NiagaraFloat=&FloatFromCurve(
     FloatCurve=DataInterface() {propertyValues {Curve {keys [
         <at 0.0041 0.1315>
         <at 0.1088 1.9664>
         {arriveTangent=-3.3757, interpMode=RCIM_Cubic, leaveTangent=-3.3757, tangentMode=RCTM_User, time=0.3609, value=0.6391}
         {arriveTangent=-1, interpMode=RCIM_Cubic, leaveTangent=-1, tangentMode=RCTM_User, time=0.5614, value=0.4386}
         <at 0.7086 1.6472>
         <at 1 0>
        ]}}}
     CurveIndex=$Emitter.NormalizedLoopAge
     Scale Curve=500
    )
   ) {isSetParameters=true}
   &EmitterState(Life Cycle Mode=::NewEnumerator1, Loop Duration=0.05, UseLoopDelay=false)
   &SpawnBurst_Instantaneous(
    Spawn Count=&RandomRangeInt(Minimum=10, Maximum=50)
    Use Spawn Probability=false
    UseLoopCountLimit=false
   )
   &SpawnRate(
    SpawnRate=&Multiply_Float(
     A=&Multiply_Float(A=9000, B=&Lerp_Float(A=0.25, B=4, Alpha=$NPC.FortniteNPC.WindGustsSystem_Active))
     B=&FloatFromCurve(
      FloatCurve=DataInterface() {propertyValues {Curve {
         defaultValue=1
         keys [
          {interpMode=RCIM_Cubic, tangentMode=RCTM_User, time=0, value=1}
          {interpMode=RCIM_Cubic, tangentMode=RCTM_User, time=0.5, value=0.02}
          {tangentMode=RCTM_User, time=1, value=1}
         ]
        }}}
      CurveIndex=Linked() {linkedVariable {name=NPC.FortniteNPC.TimeOfDay_0-1, type=NiagaraFloat}}
      Scale Curve=0.3
     )
    )
    Use Spawn Probability=false
   ) {module=SpawnRate001}
  ]
  particleSpawnScript [
   /SetVariables(Particles.MeshIndex=&RandomRangeInt(Minimum=0, Maximum=1)) {isSetParameters=true}
   /SetVariables(Particles.HeightOffset:NiagaraFloat=&RandomRangeFloat(Minimum=17500, Maximum=3000)) {isSetParameters=true}
   &InitializeParticle(
    Color Maximum=<rgba 0.0833 0.0257 0.0295 1>
    Color Minimum=<rgba 0.1952 0.0325 0.0359 1>
    Color Mode=::NewEnumerator2
    Lifetime=&UniformRangedFloat(Minimum=3, Maximum=6)
    Mass Mode=::NewEnumerator1
    Mesh Scale=&Multiply_VectorByFloat(
     Vector=&RandomRangeVector(Minimum=<2 1.3 1>, Maximum=<2.5 1.5 1>)
     Float=&RandomRangeFloat(Minimum=1, Maximum=1.5)
    )
    Mesh Scale Mode=::NewEnumerator4
    UsePositionOffset=false
    Write Material Random=true
   ) {module=InitializeParticle001}
   &AddVelocity(Velocity=&Multiply_VectorByFloat(
     Vector=&Multiply_VectorByFloat(
      Vector=&Multiply_VectorByFloat(Vector=&UniformRangedVector(Minimum=<30 50 -250>, Maximum=<80 150 500>), Float=$Emitter.SpawnSpeed_Curve)
      Float=$Emitter.Overall_SpeedMult
     )
     Float=1
    ))
   &ShapeLocation(
    MaskSpawnGroup=false
    Non Uniform Scale=&RandomRangeVector(Minimum=<-1 -1 -1>, Maximum=<1 1 1>)
    Offset=<0 0 1500>
    Offset Mode=::NewEnumerator0
    Override Seed=false
    Sphere Radius=7000
    UseOwnerScale=true
   )
   /Niagara/Modules/Spawn/Location/SphereLocation(
    MaskSpawnGroup=false
    Module.Override Randomness=false
    Non Uniform Scale=&RandomRangeVector(Minimum=<-1 -1 -1>, Maximum=<1 1 1>)
    Offset=<0 0 1500>
    Sphere Radius=7000
    Surface Only=false
   ) {enabled=false, module=SphereLocation001}
   &KillParticles() {module=KillParticles002}
   /SetVariables(Particles.InitialPos:NiagaraPosition=$Particles.Position) {isSetParameters=true}
   /Niagara/Modules/Math/PlaceParticlesOnDepthBuffer_GPU(Max Z Depth=10000, Module.Use Max Z Depth=true)
   &KillParticles(Kill Particles=&SetBoolByFloatComparison(
     A=&VectorLength(Vector=&Subtract_Vector(A=$Particles.InitialPos, B=Linked() {linkedVariable {name=Particles.Position, type=NiagaraPosition}}))
     B=1500
    ))
   &KillParticles(Kill Particles=&Boolean_NotOperation(Bool=$Output.PlaceParticlesOnDepthBuffer_GPU.ValidSample)) {module=KillParticles001}
   /Game/Effects/Niagara/Modules/Spawn/FN_Spawn_GBuffer_Color_WithinRangeV2(HueMin=345) {enabled=false}
   &FN_Spawn_GBuffer_Color_WithinRangeV2(HueMax=65, HueMin=53) {module=FN_Spawn_GBuffer_Color_WithinRangeV2001}
   &CurlNoiseForce(
    Noise Frequency=25
    Noise Strength=&Multiply_Float(
     A=250
     B=&Multiply_Float(A=&Multiply_Float(A=1, B=$Emitter.Overall_SpeedMult), B=1)
    )
    Pan Noise=false
   ) {module=CurlNoiseForce001}
   &ApplyInitialForces()
   &CameraOffset(Camera Offset Amount=-200)
   &Color(Color=&RandomRangeLinearColor(Minimum=<rgba 0.0497 0.0482 0.0319 1>, Maximum=<rgba 0.0356 0.0369 0.0194 1>)) {enabled=false}
  ]
  particleUpdateScript [
   /SetVariables(Particles.ParticleUpdate_RotRate:NiagaraFloat=&FloatFromCurve(
     FloatCurve=DataInterface() {propertyValues {Curve {keys [
         <at 0.0041 0.1315>
         <at 0.1088 1.9664>
         {arriveTangent=-3.3757, interpMode=RCIM_Cubic, leaveTangent=-3.3757, tangentMode=RCTM_User, time=0.3609, value=0.6391}
         {arriveTangent=-1, interpMode=RCIM_Cubic, leaveTangent=-1, tangentMode=RCTM_User, time=0.5614, value=0.4386}
         <at 0.7086 1.6472>
         {arriveTangent=-1.004, interpMode=RCIM_Cubic, leaveTangent=-1.004, tangentMode=RCTM_User, time=0.8311, value=0.9545}
         {arriveTangent=-5.6531, interpMode=RCIM_Cubic, leaveTangent=-5.6531, tangentMode=RCTM_User, time=0.8987, value=1.4563}
         <at 0.9384 0.3483>
         <at 1 0>
        ]}}}
     Scale Curve=5
    )) {isSetParameters=true}
   &ParticleState()
   &UpdateMeshOrientation(Rotation Rate=&FloatFromCurve(
     FloatCurve=DataInterface() {propertyValues {Curve {keys [<at 250 1>,<at 500 4>]}}}
     CurveIndex=&MakeFloatFromVector(Vector=$Particles.Velocity, Channel=::NewEnumerator0)
    ))
   &ScaleMeshSize(Scale Factor=&Multiply_VectorByFloat(
     Vector=<1 1 1>
     Float=&FloatFromCurve(FloatCurve=DataInterface() {propertyValues {Curve {keys [
          {interpMode=RCIM_Cubic, tangentMode=RCTM_User, time=0, value=0}
          {arriveTangent=2, leaveTangent=2, time=0.1, value=1}
          <at 0.5 1>
          {arriveTangent=-2, leaveTangent=-2, time=0.8, value=1}
          {tangentMode=RCTM_User, time=1, value=0}
         ]}}})
    ))
   /Niagara/Modules/Update/SubUV/SubUVAnimation(Number Of Frames=64)
   &ScaleSpriteSize(Scale Factor=&Multiply_Vector2(
     A=<1 1>
     B=&Vector2DFromFloat(Value=&Multiply_Float(
       A=&FloatFromCurve(FloatCurve=DataInterface() {propertyValues {Curve {keys [
            {interpMode=RCIM_Cubic, tangentMode=RCTM_User, time=-0.0061, value=0.489}
            {arriveTangent=0.3677, interpMode=RCIM_Cubic, leaveTangent=0.3677, tangentMode=RCTM_User, time=0.2892, value=0.9107}
            {arriveTangent=-1.2812, interpMode=RCIM_Cubic, leaveTangent=-1.2812, tangentMode=RCTM_User, time=0.9809, value=0.9275}
            <at 1 0>
           ]}}})
       B=&FloatFromCurve(
        FloatCurve=DataInterface() {propertyValues {Curve {keys [
            {interpMode=RCIM_Cubic, tangentMode=RCTM_User, time=0, value=0.0005}
            {arriveTangent=0.0074, interpMode=RCIM_Cubic, leaveTangent=0.0074, tangentMode=RCTM_User, time=0.15, value=1.0236}
            {arriveTangent=-0.0952, interpMode=RCIM_Cubic, leaveTangent=-0.0952, tangentMode=RCTM_User, time=0.85, value=1}
            <at 0.9 0>
           ]}}}
        CurveIndex=&NormalizeDistanceRange(
         Start Position:NiagaraPosition=Linked() {linkedVariable {name=Emitter.CameraPostion, type=Vector3f}}
         End Position:NiagaraPosition=$Particles.Position
         Distance:NiagaraFloat=10000
         Invert Normalized Range:NiagaraBool=false
        )
        Scale Curve=0.85
       )
      ))
    )) {module=ScaleSpriteSize001}
   &Drag(Use Linear Drag=true, Use Rotational Drag=false)
   &CurlNoiseForce(
    Curl Noise Cone Mask Angle=50
    Mask Curl Noise=true
    Noise Frequency=&RandomRangeFloat(Minimum=5, Maximum=25)
    Noise Strength=&Multiply_Float(A=$Emitter.SpawnSpeed_Curve, B=500)
    Pan Noise=true
    Pan Noise Field=<0 0 1>
    Random Seed=80
   )
   &ScaleColor(
    Scale Alpha=&FloatFromCurve(
     FloatCurve=DataInterface() {propertyValues {Curve {keys [
         {interpMode=RCIM_Cubic, tangentMode=RCTM_User, time=0, value=0}
         {arriveTangent=-0.0948, interpMode=RCIM_Cubic, leaveTangent=-0.0948, tangentMode=RCTM_User, time=0.1, value=1}
         <at 0.942 0.9825>
         <at 1 0>
        ]}}}
     Scale Curve=&FloatFromCurve(FloatCurve=DataInterface() {propertyValues {Curve {keys [
          {interpMode=RCIM_Cubic, tangentMode=RCTM_User, time=0, value=0}
          {interpMode=RCIM_Cubic, tangentMode=RCTM_User, time=0.2, value=5}
          {tangentMode=RCTM_User, time=1, value=0}
         ]}}})
    )
    ScaleA=true
    ScaleRGB=true
   )
   /Niagara/Modules/Update/SubUV/V2/SubUVAnimation(
    Start Frame:NiagaraInt32=&RandomRangeInt(Minimum=0, Maximum=32)
    SubUV Animation Mode:ENiagara_SubUVLookupMode=::NewEnumerator1
    SubUV Loop Count:NiagaraFloat=&SelectFloatFromArray(
     Array Sampling Mode=::NewEnumerator0
     Direct Array Index=?
     Float Selection Array=DataInterface() {propertyValues {FloatData [2,4]}}
     Random Seed=0
    )
   ) {module=SubUVAnimation001}
   &DynamicMaterialParameters(
    Index 0 Param 1=&RandomRangeFloat(Minimum=0.1, Maximum=0.15)
    Index 0 Param 2=&RandomRangeFloat(Minimum=0, Maximum=0)
    Module.Param0WriteEnabled=true
    Module.Param1WriteEnabled=true
    Module.Param2WriteEnabled=true
    Module.Param3WriteEnabled=true
   ) {enabled=false, module=DynamicMaterialParameters001}
   &GravityForce(Gravity=&Multiply_VectorByFloat(Vector=<0 0 -980>, Float=&RandomRangeFloat(Minimum=1, Maximum=2)))
   &WindForce(
    Friction=2
    Ground Distance Min / Max:Vector2f=<0 100>
    Ground Elevation:NiagaraFloat=&MakeFloatFromVector(Vector=Linked() {linkedVariable {name=Engine.Emitter.SimulationPosition, type=NiagaraPosition}}, Channel=::NewEnumerator2)
    Scale Min / Max:Vector2f=<0.5 1>
    Speed Min / Max:Vector2f=<0 150>
    Turbulence Scale Mode:ENiagaraWindTurbulenceContributionMode=::NewEnumerator3
    Use Ground Mask=true
    Wind Speed=<250 0 -100>
    Wind Speed Scale=&Lerp_Float(A=0.25, B=1, Alpha=$NPC.FortniteNPC.WindGustsSystem_Active)
   ) {module=WindForce001}
   &AerodynamicDrag(Aerodynamic Drag=0.5, Use Linear Drag=true, Use Rotational Drag=true)
   &SolveForcesAndVelocity(Clamp Velocity=false, Limit Acceleration=false)
   &AlignSpriteToMeshOrientation(Mesh Orientation Relative Sprite Facing Vector:Vector3f=<0 0 0>, SetAlignment=true, SetFacing=true)
   &DynamicMaterialParameters(
    Index 0 Param 1=&RandomRangeFloat(Minimum=0.01, Maximum=0.015)
    Module.Index0Float0Write=true
    Module.Index0Float1Write=true
    Module.Index0Float2Write=true
    Module.Index0Float3Write=true
   ) {module=DynamicMaterialParameters002}
  ]
  rendererClasses [&NiagaraMeshRendererProperties]
  renderers [
   {rendererClass=&NiagaraMeshRendererProperties, rendererIndex=0}
   {rendererClass=&NiagaraMeshRendererProperties, rendererIndex=1}
  ]
 }
}
summary {
 systemName=NS_FNM_Gourdy_Atmosphere
 emitters [
  {
   emitterName=Fog
   simTarget=GPUComputeSim
   rendererClasses [&NiagaraSpriteRendererProperties, &NiagaraMeshRendererProperties]
  }
  {emitterName=Bats001, simTarget=GPUComputeSim, rendererClasses [&NiagaraMeshRendererProperties]}
  {emitterName=LeavesInAir001, simTarget=GPUComputeSim, rendererClasses [&NiagaraMeshRendererProperties]}
 ]
 userVariables [
  User.Landscape {
   type=NiagaraDataInterfaceLandscape
   description=""
   defaultValue {
    struct=NiagaraExt_VariableValue_DataInterface
    value {
     dataInterface=&self:NiagaraDataInterfaceLandscape_0
     dataInterfaceClass=&NiagaraDataInterfaceLandscape
    }
   }
  }
  User.Radius Mult {type=NiagaraFloat, defaultValue=1, description=""}
  User.Spawn at Day {type=NiagaraFloat, defaultValue=1, description=""}
  User.Spawn at Evening {type=NiagaraFloat, defaultValue=1, description=""}
  User.Spawn at Morning {type=NiagaraFloat, defaultValue=1, description=""}
  User.Spawn at Night {type=NiagaraFloat, defaultValue=1, description=""}
  User.Spawn Rate Mult {type=NiagaraFloat, defaultValue=1, description=""}
 ]
}
